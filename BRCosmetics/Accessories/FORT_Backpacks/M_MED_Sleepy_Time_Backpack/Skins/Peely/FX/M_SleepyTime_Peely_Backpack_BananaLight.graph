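# Translucent, unlit, two-sided Niagara mesh-particle material for the banana light effect on the Peely Sleepy Time backpack.
# Opacity logic multiplies an elements texture by panning cloud noise and swirl noise sampled from time plus vertex red.
# Branches add camera-facing contrast with a simpler low feature-level path, and a sine pulse with particle alpha.
# Day/night color tinted by particle color feeds emissive adjustments; swirl noise also scales a normal-direction offset.
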Material /BRCosmetics/Accessories/FORT_Backpacks/M_MED_Sleepy_Time_Backpack/Skins/Peely/FX/M_SleepyTime_Peely_Backpack_BananaLight
graph {
 properties {
  bUsedWithNiagaraMeshParticles=true
  blendMode=BLEND_Translucent
  pixelDepthOffsetMode=PDOM_Legacy
  shadingModel=MSM_Unlit
  twoSided=true
  usedShadingModels=Unlit
 }
 nodes [
  n1 = DotProduct(A=CameraVectorWS(), B=VertexNormalWS())
  n2 = TextureCoordinate()
  n3 = TextureSample(UVs=StaticSwitchParameter(True=Multiply(A=@n2) {constB=-1}, False=@n2) {defaultValue=true, parameterName=flip_Texture}) {texture=&/T_SleepyTime_Peely_Elements}
  n4 = ScalarParameter() {defaultValue=-1, parameterName=Param_3}
  n5 = ScalarParameter() {defaultValue=2, parameterName=Param_2}
  n6 = Multiply(A=Time()) {constB=0.03}
  n7 = VertexColor()
  n8 = Multiply(A=Time()) {constB=0.01}
  n9 = TextureSample(UVs=AppendVector(A=Add(A=@n6, B=@n7.R), B=Add(A=@n8, B=@n7.R))) {samplerType=SAMPLERTYPE_LinearGrayscale, texture=/Game/Global/Textures/Default/Noise/T_SwirlNoise_N_height}
  n10 = LinearInterpolate(A=@n3.R, B=@n3.R, Alpha=Clamp(None=LinearInterpolate(A=@n4, B=@n5, Alpha=@n9.R)))
  n11 = VectorParameter() {parameterName=Noise3_texture, defaultValue=<rgba 0.4 0.4 -0.1 0.05>}
  n12 = Multiply(A=Panner(Speed=AppendVector(A=@n11.B, B=@n11.B)), B=AppendVector(A=@n11.R, B=@n11.R))
  n13 = Constant() {r=1}
  n14 = MaterialFunctionCall(
   Input=TextureSample(UVs=Multiply(A=@n12) {constB=0.5}) {texture=&AtmosphericCloudNoise01}.R
   Input Low=Constant()
   Input High=@n13
   Target Low=Constant() {r=0.4}
   Target High=@n13
  ) {materialFunction=&RemapValueRange}
  n15 = Multiply(A=ScalarParameter() {defaultValue=2, parameterName=BananaOpacity}, B=Multiply(A=Multiply(A=@n10, B=@n9.R), B=Power(Base=@n14.Result)))
  n16 = Multiply(
   A=MaterialFunctionCall(In=Abs(None=@n1), Contrast=Constant() {r=1}) {materialFunction=&CheapContrast}.Result
   B=@n15
  )
  FeatureLevelSwitch(Default=@n16, ES2=@n15, ES3_1=@n15, SM4_REMOVED=@n16, SM5=@n16, SM6=@n16)
  MaterialFunctionCall(In=@n9.R, Contrast=ScalarParameter() {defaultValue=10, parameterName=Param_6}) {materialFunction=&CheapContrast}
  n17 = VectorParameter() {parameterName=DayColor, defaultValue=<rgba 1.5 1.0157 0.3359 1>}
  n18 = VectorParameter() {parameterName=NightColor, defaultValue=<rgba 1.5 0.9456 0.315 1>}
  n19 = ParticleColor()
  MaterialFunctionCall(Emissive=Multiply(
    A=MaterialFunctionCall(Color_Morning=@n17.RGB, Color_Day=@n17.RGB, Color_Evening=@n17.RGB, Color_Night=@n18.RGB) {materialFunction=&Color_Time_of_Day}.Result
    B=@n19.RGB
   )) {materialFunction=&MF_FX_EmissiveAdjustments}
  Multiply(
   A=Multiply(A=VertexNormalWS(), B=ScalarParameter() {defaultValue=0.5, parameterName=Normal_intensity})
   B=LinearInterpolate(A=ScalarParameter() {defaultValue=-2.5, parameterName=Param_8}, B=ScalarParameter() {defaultValue=0.8, parameterName=Param_7}, Alpha=@n9.R) {constA=-2, constB=0.8}
  )
  n20 = Multiply(A=Time()) {constB=0.2}
  n21 = Constant() {r=1}
  Multiply(
   A=MaterialFunctionCall(Input=Sine(None=@n20), Input Low=Constant() {r=-1}, Input High=@n21, Target Low=Constant() {r=0.6}, Target High=@n21) {materialFunction=&RemapValueRange}.Result
   B=Multiply(A=@n19.A, B=@n15)
  )
  Reroute()
 ]
}
