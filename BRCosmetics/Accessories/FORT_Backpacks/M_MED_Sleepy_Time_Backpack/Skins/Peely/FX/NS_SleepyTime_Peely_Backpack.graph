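# Niagara idle effect for the Peely skin of the Sleepy Time backpack cosmetic, producing a warm glow.
# A single Idle emitter burst-spawns particles with a fixed mesh orientation and disables killing at lifetime end.
# Mesh, sprite and light renderers draw them, with a camera offset attribute for the glow sprite.
# A Gen5 light color attribute keeps its warm RGB while a normalized-age curve dips and restores its alpha.
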NiagaraSystem /BRCosmetics/Accessories/FORT_Backpacks/M_MED_Sleepy_Time_Backpack/Skins/Peely/FX/NS_SleepyTime_Peely_Backpack
dependencies {usedDataInterfaces [&NiagaraDataInterfaceColorCurve], usedDynamicInputs [], usedModules [&SystemState]}
emitters {Idle {
  emitterUpdateScript [
   &EmitterState(
    Inactive Response=::NewEnumerator1
    Life Cycle Mode=::NewEnumerator1
    Loop Behavior=::NewEnumerator1
    Loop Duration Mode=::NewEnumerator1
    UseLoopDelay=false
   )
   &SpawnBurst_Instantaneous(Use Spawn Probability=false, UseLoopCountLimit=false)
  ]
  particleSpawnScript [
   &InitializeParticle(Color=<rgba 1 0.6715 0.1375 2>, Lifetime=3, UsePositionOffset=false, Write Material Random=true)
   &InitialMeshOrientation(Mesh Orientation Mode=::NewEnumerator2, Orientation Vector=<0 1 0>, Use Rotation Vector=true)
  ]
  particleUpdateScript [
   &ParticleState(Kill Particles When Lifetime Has Elapsed=false)
   /SetVariables(
    Particles.Gen5LightColor:LinearColor=&ScaleLinearColorByCurve(
     InitialValue=<rgba 1 0.7244 0.2645 1>
     ScaleCurve=DataInterface() {propertyValues {
       AlphaCurve {keys [
         {interpMode=RCIM_Cubic, tangentMode=RCTM_User, time=0, value=5}
         {interpMode=RCIM_Cubic, tangentMode=RCTM_User, time=0.2243, value=4}
         <at 1 5>
        ]}
       BlueCurve {keys [<at 0 1>,<at 1 1>]}
       GreenCurve {keys [<at 0 1>,<at 1 1>]}
       RedCurve {keys [<at 0 1>,<at 1 1>]}
      }}
     CurveIndex=$Particles.NormalizedAge
    )
    Particles.GlowSpritecameraOffset=15
   ) {isSetParameters=true}
  ]
  rendererClasses [
   &NiagaraMeshRendererProperties
   &NiagaraSpriteRendererProperties
   &NiagaraLightRendererProperties
  ]
  renderers [
   {rendererClass=&NiagaraMeshRendererProperties, rendererIndex=0}
   {rendererClass=&NiagaraSpriteRendererProperties, rendererIndex=1}
   {rendererClass=&NiagaraLightRendererProperties, rendererIndex=2}
  ]
 }}
summary {
 systemName=NS_SleepyTime_Peely_Backpack
 emitters [{
   emitterName=Idle
   rendererClasses [
    &NiagaraMeshRendererProperties
    &NiagaraSpriteRendererProperties
    &NiagaraLightRendererProperties
   ]
  }]
 userVariables []
}
